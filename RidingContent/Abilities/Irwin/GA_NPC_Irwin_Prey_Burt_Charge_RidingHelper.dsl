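; Gameplay ability that adjusts the Irwin prey NPC Burt's charge attack for riding, triggered by gameplay events and switched on event tag.
; On override start it caches pawn and charge ability, forces forward movement, changes speed, acceleration and yaw rate, fires a rider-dependent boost cue, then ends.
; StoppedRiding ends the cached charge early.
; Charge-end logic clears forced forward and restores speed and acceleration, leaving rotation reset to the charge ability.
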(event Ability|EventActivateAbilityFromEvent
  (params (EventData :type Gameplay%20Event%20Data%20Structure%20%28by%20ref%29))
  (bind (_eventtag _instigator _target _optionalobject) (Utilities|Struct|BreakGameplayEventData EventData))
  (bind _controllableridablecomponent_cached (Variables|Default|GetControllableRidableComponent_Cached))
  (bind _targetabilitycached (Variables|Default|GetTargetAbilityCached))
  (bind _controllableridablecomponent_cached_2 (Variables|Default|GetControllableRidableComponent_Cached))
  (bind _donotuse_primaryrider (Variables|/Script/RidingCodeRuntime.RidableComponent|GetDoNotUse_PrimaryRider _controllableridablecomponent_cached_2))
  (bind _fortaipawncached (Variables|Default|GetFortAIPawnCached))
  (switch Utilities|FlowControl|Switch|SwitchonGameplayTag _eventtag
    (:NPC.Ability.Attack.Melee.ChargeControl.OverrideStart
      ;[1 1 1] Adjustments On Charge Begin
      (bind _asfort_aipawn (Utilities|Casting|CastToFortAIPawn (Ability|GetAvatarActorfromActorInfo)))
      (Variables|Default|SetFortAIPawnCached _asfort_aipawn)
      (Utilities|IsValid _optionalobject
        (:Is%20Valid
          (bind _asfort_gameplay_ability (Utilities|Casting|CastToFortGameplayAbility _optionalobject))
          (Variables|Default|SetTargetAbilityCached _asfort_gameplay_ability)
          (Utilities|FlowControl|Sequence
            (:then_0
              ;;[1 1 1] Force Forward
              (Utilities|IsValid (Actor|GetComponentbyClass _asfort_aipawn "/Script/RidingCodeRuntime.ControllableRidableComponent")
                (:Is%20Valid
                  (bind _output_get (Variables|Default|SetControllableRidableComponent_Cached (Actor|GetComponentbyClass _asfort_aipawn "/Script/RidingCodeRuntime.ControllableRidableComponent")))
                  (Variables|/Script/RidingCodeRuntime.ControllableRidableComponent|SetbAlwaysMoveForward true _output_get))))
            (:then_1
              ;;[1 1 1] Set Speed and Control Adjustments
              (Utilities|IsValid (Variables|/Script/Engine.Character|GetCharacterMovement _asfort_aipawn)
                (:Is%20Valid
                  (FortPawn|SetMovementSpeedMultiplier _asfort_aipawn 2.3)
                  (Variables|Default|SetInitialMaxAcceleration (Variables|/Script/Engine.CharacterMovementComponent|GetMaxAcceleration (Variables|/Script/Engine.Character|GetCharacterMovement _asfort_aipawn)))
                  (Variables|/Script/Engine.CharacterMovementComponent|SetMaxAcceleration 1.5e+03 (Variables|/Script/Engine.Character|GetCharacterMovement _asfort_aipawn))
                  (FortPawn|SetAIRotationRate _asfort_aipawn (Math|Rotator|MakeRotator (.roll (Variables|/Script/Engine.CharacterMovementComponent|GetRotationRate (Variables|/Script/Engine.Character|GetCharacterMovement _asfort_aipawn))) (.pitch (Variables|/Script/Engine.CharacterMovementComponent|GetRotationRate (Variables|/Script/Engine.Character|GetCharacterMovement _asfort_aipawn))) (Variables|Default|GetMaxRotationRateYaw))))))
            (:then_2
              ;;[1 1 1] Fire Boost Effects and End Ability
              (Utilities|IsValid _controllableridablecomponent_cached_2
                (:Is%20Valid
                  (Utilities|IsValid _donotuse_primaryrider
                    (:Is%20Valid
                      (Utilities|IsValid (Components|GetOwner _donotuse_primaryrider)
                        (:Is%20Valid
                          (Ability|ExecuteGameplayCueOnOwner (Variables|Default|GetGCN_BoostTag)))))))))
            (:then_3
              (Ability|EndAbility))))))
    (:NPC.Ability.Attack.Melee.ChargeControl.OverrideEnd
      _)
    (:Gameplay.Action.Player.StoppedRiding
      ;[1 1 1] If riding ended, stop ability early
      (Utilities|IsValid _targetabilitycached
        (:Is%20Valid
          (Ability|EndAbility _targetabilitycached)))))
  ;[1 1 1] Adjustments For Charge End
  (Utilities|FlowControl|Sequence
    (:then_0
      ;;[1 1 1] Set Force Forward Off
      (Utilities|IsValid _controllableridablecomponent_cached
        (:Is%20Valid
          (Variables|/Script/RidingCodeRuntime.ControllableRidableComponent|SetbAlwaysMoveForward false _controllableridablecomponent_cached))))
    (:then_1
      ;;[1 1 1] Reset Movement Speed And Control Adjustments (NOTE: Rotation is not here, the charge ability resets it)
      (Utilities|IsValid _fortaipawncached
        (:Is%20Valid
          (FortPawn|SetMovementSpeedMultiplier _fortaipawncached 1.0)
          (Variables|/Script/Engine.CharacterMovementComponent|SetMaxAcceleration (Variables|Default|GetInitialMaxAcceleration) (Variables|/Script/Engine.Character|GetCharacterMovement _fortaipawncached)))))
    (:then_2
      (Ability|EndAbility))))
